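# Two-sided material for the AlleyTown hanging wires prop, usable on Nanite and spline meshes, with wind sway and adjustable thickness.
# Two summed sines sway the wire along WindDirection, using vertex red for a midspan-symmetric falloff, blue for variation and alpha for amplitude.
# SplineThicken adds width, optionally shrunk to zero by pixel depth.
# Its normal blends with a detail map, alongside tiled diffuse and packed specular/metal/roughness textures.
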Material /Hermes/Props/AlleyTown/AlleyTown_HangingWires_A/Materials/M_AlleyTown_HangingWires_A
graph {
 properties {bUsedWithNanite=true, bUsedWithSplineMeshes=true, pixelDepthOffsetMode=PDOM_Legacy, twoSided=true, usedShadingModels=Default Lit}
 nodes [
  n1 = VertexColor()
  n2 = Add(A=Time(), B=@n1.B)
  n3 = Add(
   A=ScalarParameter() {defaultValue=0.13, parameterName=WindSpeed}
   B=Multiply(A=ScalarParameter() {defaultValue=0.01, parameterName=WindSpeedVariationAdd2}, B=@n1.B)
  )
  n4 = ScalarParameter() {defaultValue=0.195, parameterName=WindSpeedVariationAdd}
  n5 = Sine(None=Multiply(A=@n2, B=Add(A=@n3, B=@n4)))
  n6 = Multiply(A=@n2, B=@n3)
  n7 = Multiply(A=Add(A=@n5, B=Sine(None=@n6)), B=Multiply(A=ScalarParameter() {defaultValue=9.0156, parameterName=WindAmplitude}, B=@n1.A))
  n8 = OneMinus(None=@n1.R)
  n9 = Power(Base=Multiply(A=Multiply(A=@n1.R, B=@n8)) {constB=4}, Exp=ScalarParameter() {defaultValue=-0.1822, parameterName=WindFalloff})
  n10 = VectorParameter() {parameterName=WindDirection, defaultValue=<rgba 0.05 1 0 1>}
  n11 = LinearInterpolate(
   A=ScalarParameter() {defaultValue=4, parameterName=Thickness_Max}
   B=ScalarParameter() {defaultValue=4, parameterName=Thickness_Min}
   Alpha=NamedRerouteUsage()
  )
  n12 = Divide(
   A=Add(A=PixelDepth().R, B=ScalarParameter() {defaultValue=-5500, group=DistanceFade, parameterName=DistanceFadeOffset})
   B=ScalarParameter() {defaultValue=1500, group=DistanceFade, parameterName=DistanceFadeLength}
  )
  n13 = StaticSwitchParameter(True=LinearInterpolate(A=@n11, B=Constant(), Alpha=Saturate(None=@n12)), False=@n11) {group=DistanceFade, parameterName=UseDistancefade}
  n14 = MaterialFunctionCall(WidthBase=@n13, WidthTip=@n13) {materialFunction=&SplineThicken}
  Add(A=Multiply(A=Multiply(A=@n7, B=@n9), B=Transform(None=@n10.RGB) {transformSourceType=TRANSFORMSOURCE_Instance}), B=@n14.WorldPositionOffset)
  MaterialFunctionCall!(widthTip=@n13, Flatten=ScalarParameter() {parameterName=Flatten}, widthBase=@n13) {materialFunction=/Game/Characters/Player/Male/Medium/Bodies/M_MED_Neon/Materials/MF_splineThickenM}
  n15 = Transform(None=@n14.Normal) {transformSourceType=TRANSFORMSOURCE_World, transformType=TRANSFORM_Tangent}
  n16 = ComponentMask(None=TextureSampleParameter2D(UVs=NamedRerouteUsage()) {parameterName=AdditionalNormal, samplerType=SAMPLERTYPE_Normal, texture=/EpicBaseTextures/Blanks/T_Default_NormalMap}.RGB) {b=true, g=true, r=true}
  MaterialFunctionCall!(Normal_1=@n15, Normal_2=@n16) {materialFunction=/Game/Global/MaterialFunctions/MF_NormalBlend_Whiteout}
  MaterialFunctionCall(BaseNormal=Multiply(A=@n15, B=TwoSidedSign()), AdditionalNormal=@n16) {materialFunction=&BlendAngleCorrectedNormals}
  Multiply(A=@n5) {constB=0.75}
  Multiply(A=NamedRerouteUsage(), B=ScalarParameter() {defaultValue=0.25, parameterName=SubSurfaceBrightness})
  Multiply(A=ScalarParameter() {defaultValue=5, parameterName=ThicknessOverallMult}, B=@n11)
  n17 = MaterialFunctionCall(Float2=TextureCoordinate()) {materialFunction=&BreakOutFloat2Components}
  n18 = ScalarParameter() {defaultValue=1, parameterName=Tiling_U}
  n19 = ScalarParameter() {defaultValue=1, parameterName=Tiling_V}
  NamedRerouteDeclaration(None=AppendVector(A=Multiply(A=@n17.R, B=@n18), B=Multiply(A=@n17.G, B=@n19))) {name=TextureCoord, nodeColor=<rgba 0.7109 0 1 1>}
  n20 = TextureSampleParameter2D(UVs=NamedRerouteUsage()) {parameterName=Spec, texture=&T_BlacksRGB}
  NamedRerouteDeclaration(None=@n20.B) {name=Rough, nodeColor=<rgba 0 1 0.6484 1>}
  NamedRerouteDeclaration(None=@n20.G) {name=Metal, nodeColor=<rgba 1 0.4453 0 1>}
  NamedRerouteDeclaration(None=@n20.R) {name=Specular, nodeColor=<rgba 0 1 0.5547 1>}
  n21 = VertexColor()
  NamedRerouteDeclaration(None=@n21.R) {name=VertColor_Red, nodeColor=<rgba 1 0.8438 0 1>}
  NamedRerouteDeclaration(None=@n21.G) {name=VertColor_Green, nodeColor=<rgba 0 1 0.2031 1>}
  NamedRerouteDeclaration(None=TextureSampleParameter2D(UVs=NamedRerouteUsage()) {parameterName=Diffuse, texture=&T_WhitesRGB}.RGB) {name=Diffuse, nodeColor=<rgba 0 1 0.9766 1>}
  NamedRerouteUsage()
  NamedRerouteUsage()
  NamedRerouteUsage()
  ScalarParameter() {defaultValue=0.25, parameterName=SubSurfaceOpacity}
 ]
}
